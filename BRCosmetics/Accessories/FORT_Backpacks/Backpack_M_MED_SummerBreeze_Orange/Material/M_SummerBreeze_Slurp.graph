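# Skeletal-mesh character material for the orange SummerBreeze backpack, adding a Slurp liquid effect to the standard character shading stack.
# MF_Slurp sits between the PreFX and PostFX character functions, driven by selectable SkinFX_Mask channels for slurp, bubbles and normal distortion.
# Its emissive is tinted and desaturated by time-of-day parameters, then scaled by a frontend-only boost before PostFX and a final cinematic post function.
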Material /BRCosmetics/Accessories/FORT_Backpacks/Backpack_M_MED_SummerBreeze_Orange/Material/M_SummerBreeze_Slurp
graph {
 properties {
  bUseMaterialAttributes=true
  bUsedWithMeshDeformer=true
  bUsedWithSkeletalMesh=true
  pixelDepthOffsetMode=PDOM_Legacy
  usedShadingModels=Default Lit
 }
 nodes [
  n1 = MaterialFunctionCall() {materialFunction=&BaseFN_CharacterFunction_PreFX}
  n2 = TextureSample(
   UVs=TextureCoordinate()
   Tex=NamedRerouteDeclaration(None=TextureObjectParameter() {group=Skin FX, parameterName=SkinFX_Mask}) {name=SkinFX_Mask, nodeColor=<rgba 0.9842 0 1 1>}
  )
  n3 = GetMaterialAttributes(None=MaterialFunctionCall!(
    Alpha=StaticComponentMaskParameter(None=@n2.RGBA) {defaultR=true, group=Skin FX, parameterName=Slurp Channel}
    AO=@n1.AO
    BubblesMask=StaticComponentMaskParameter(None=@n2.RGBA) {defaultG=true, group=Skin FX, parameterName=Slurp Bubbles Channel}
    SlurpNormalDistortionMask=StaticComponentMaskParameter(None=@n2.RGBA) {defaultB=true, group=Skin FX, parameterName=Slurp Normal Distortion Channel}
    Pre Skinned Position=@n1.AO
    Material=@n1.AO
   ) {materialFunction=/Game/Weapons/WeaponSkins/MaterialFunction/MF_Slurp}.Result) {attributeGetTypes [#1]}
  n4 = VectorParameter() {parameterName=Morning_TimeOfDay_Color, defaultValue=<rgba 1 1 1 0>}
  n5 = VectorParameter() {parameterName=Day_TimeOfDay_Color, defaultValue=<rgba 1 1 1 0>}
  n6 = VectorParameter() {parameterName=Evening_TimeOfDay_Color, defaultValue=<rgba 1 1 1 0>}
  n7 = VectorParameter() {parameterName=Night_TimeOfDay_Color, defaultValue=<rgba 1 1 1 0>}
  n8 = Desaturation(
   None=Multiply(
    A=@n3.Emissive Color
    B=MaterialFunctionCall(Color_Morning=@n4.RGB, Color_Day=@n5.RGB, Color_Evening=@n6.RGB, Color_Night=@n7.RGB) {materialFunction=&Color_Time_of_Day}.Result
   )
   Fraction=MaterialFunctionCall(
    Value_Morning=ScalarParameter() {parameterName=Morning_Time_Desaturation}
    Value_Day=ScalarParameter() {parameterName=Day_Time_Desaturation}
    Value_Evening=ScalarParameter() {parameterName=Evening_Time_Desaturation}
    Value_Night=ScalarParameter() {parameterName=Night_Time_Desaturation}
   ) {materialFunction=&Scalar_Time_of_Day}.Result
  )
  n9 = LinearInterpolate(
   A=Constant() {r=1}
   B=ScalarParameter() {defaultValue=1, parameterName=FrontEnd_Boost}
   Alpha=ScalarParameter() {parameterName=IsFrontend, sliderMax=1}
  )
  n10 = MaterialFunctionCall(
   AO=@n1.AO
   Distance Blend=@n1.AO
   Fresnel=@n1.AO
   Gmap=@n1.AO
   Pre Skinned Local Position=@n1.AO
   Material=SetMaterialAttributes(MaterialAttributes=@n3.MaterialAttributes, Emissive Color=Multiply(A=@n8, B=@n9)) {attributeSetTypes [#1]}
  ) {materialFunction=&BaseFN_CharacterFunction_PostFX}
  MaterialFunctionCall(In=@n10.Result) {materialFunction=&CINE_MF_CharPostFX}
 ]
}
